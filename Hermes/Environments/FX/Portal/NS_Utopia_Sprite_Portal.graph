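# Niagara effect for the Utopia sprite portal, with user parameters for open state and interior color.
# A system timeline drives the transition: the camera-facing portal mesh and sprite scale up, flares flash briefly, and idle glow fades.
# RadialFlow adds spinning two-layer wave sprites with biased random spin and radial material animation.
# Distance-culled motes spawn on a timeline-widened sphere and swirl under vortex and point forces, some with a mid-life sparkle.
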NiagaraSystem /Hermes/Environments/FX/Portal/NS_Utopia_Sprite_Portal
dependencies {
 usedDataInterfaces [&NiagaraDataInterfaceCurve, &NiagaraDataInterfaceCamera]
 usedDynamicInputs [
  &LerpLinearColors
  &RandomRangeLinearColor
  &InvertBool
  &MakeCustomFloatFromBool
 ]
 usedModules [
  &SystemState
  &self:SystemSpawnScript.SystemScriptSource.SystemScriptGraph.NiagaraNodeAssignment_0.SetVariables_#1
  &Timeline
 ]
}
emitters {
 Em_Idle_Glow {
  emitterUpdateScript [
   &EmitterState(
    Inactive Response=::NewEnumerator1
    Life Cycle Mode=::NewEnumerator1
    Loop Behavior=::NewEnumerator1
    Loop Duration Mode=::NewEnumerator1
    UseLoopDelay=false
   )
   &SpawnBurst_Instantaneous(Spawn Count=1, Use Spawn Probability=false)
  ]
  particleSpawnScript [
   &InitializeParticle(
    Color=<rgba 0.0267 0.3123 1 1>
    Sprite Size Mode=::NewEnumerator3
    Uniform Sprite Size=200
    UsePositionOffset=false
    Write Material Random=true
   )
   &CameraOffset(Camera Offset Amount=15)
  ]
  particleUpdateScript [
   &ParticleState(Kill Particles When Lifetime Has Elapsed=false)
   &ScaleColor(
    Scale Alpha=&FloatFromCurve(
     FloatCurve=DataInterface() {propertyValues {Curve {
        defaultValue=1
        keys [{interpMode=RCIM_Cubic, tangentMode=RCTM_User, time=0, value=1}, {interpMode=RCIM_Cubic, tangentMode=RCTM_User, time=1, value=0}]
       }}}
     CurveIndex=$System.Timeline.NormalizedTime
    )
    ScaleA=true
    ScaleRGB=false
   )
   &DynamicMaterialParameters(
    Index 0 Param 1=0.89
    Index 0 Param 2=0.8
    Module.Index0Float0Write=true
    Module.Index0Float1Write=true
    Module.Index0Float2Write=true
    Module.Index0Float3Write=true
   )
  ]
  rendererClasses [&NiagaraSpriteRendererProperties]
  renderers [
   {rendererClass=&NiagaraSpriteRendererProperties}
   {rendererClass=&NiagaraSpriteRendererProperties, rendererIndex=1}
  ]
 }
 Em_Motes {
  emitterUpdateScript [
   &EmitterState(
    Enable Distance Culling=true
    MaxDistance=25000
    Module.bUseMaxDistance=true
    Module.bUseMinDistance=false
    Scalability Mode=::NewEnumerator1
   )
   &SpawnRate(SpawnRate=5, Use Spawn Probability=false)
  ]
  particleSpawnScript [
   /SetVariables(Particles.SparkIntensity:NiagaraFloat=&FloatFromCurve(
     FloatCurve=DataInterface() {propertyValues {Curve {keys [
         <at 0 1 RCIM_Cubic>
         {arriveTangent=-2.6566, interpMode=RCIM_Cubic, leaveTangent=-2.6566, tangentMode=RCTM_User, time=0.3, value=0.55}
         <at 1 0 RCIM_Cubic>
        ]}}}
     CurveIndex=&RandomRangeFloat(Minimum=0, Maximum=1)
    )) {isSetParameters=true}
   &InitializeParticle(
    Color Maximum=<rgba 2.6557 11.4187 25 4>
    Color Minimum=<rgba 12.2752 1.8772 18 2>
    Color Mode=::NewEnumerator2
    Lifetime Max=12
    Lifetime Min=3
    Lifetime Mode=::NewEnumerator1
    Mesh Scale Mode=::NewEnumerator4
    Position Mode=::NewEnumerator1
    Sprite Rotation Mode=::NewEnumerator2
    Sprite Size Mode=::NewEnumerator3
    Uniform Sprite Size=&Lerp_Float(A=7.5, B=15, Alpha=$Particles.SparkIntensity)
    UsePositionOffset=false
    Write Material Random=true
   )
   &SphereLocation(
    MaskSpawnGroup=false
    Override Seed=false
    Sphere Radius=&Lerp_Float(A=30, B=90, Alpha=$System.Timeline.NormalizedTime)
    Surface Only=true
    UseNonUniformScale=false
    UseOffset=false
    UseSphereOrientation=false
   ) {module=SphereLocation001}
   /SetVariables(Particles.Sparkle:NiagaraFloat=&UniformAOrBFloat(A=0, Distribution Weight=0.2)) {isSetParameters=true}
   &AddVelocity(
    Scale Added Velocity=&VectorFromFloat(Value=0)
    Velocity=&RotateVectorByQuaternion(
     Vector To Rotate=&NormalizeVector(Vector To Normalize=&Subtract_Vector(A=Linked() {linkedVariable {name=Particles.Position, type=NiagaraPosition}}, B=<1200 0 500>))
     Quaternion=<0 -1 0 1>
    )
   )
  ]
  particleUpdateScript [
   &ParticleState()
   &DynamicMaterialParameters(
    Index 0 Param 1=1
    Index 0 Param 2=1
    Index 0 Param 3=0.5
    Module.Param0WriteEnabled=true
    Module.Param1WriteEnabled=true
    Module.Param2WriteEnabled=true
    Module.Param3WriteEnabled=true
   ) {enabled=false}
   &ScaleColor(
    Scale Alpha=&FloatFromCurve(FloatCurve=DataInterface() {propertyValues {Curve {keys [
         <at 0 0 RCIM_Cubic>
         {interpMode=RCIM_Cubic, tangentMode=RCTM_User, time=0.1, value=0.9914}
         {interpMode=RCIM_Cubic, tangentMode=RCTM_User, time=0.7, value=1}
         <at 1 0>
        ]}}})
    ScaleA=true
    ScaleRGB=true
   )
   &ScaleSpriteSize(Scale Factor=&Vector2DFromFloat(Value=&FloatFromCurve(
      FloatCurve=DataInterface() {propertyValues {Curve {keys [
          <at 0 0 RCIM_Cubic>
          {arriveTangent=2.4312, interpMode=RCIM_Cubic, leaveTangent=2.4312, time=0.1483, value=1}
          {arriveTangent=-0.0468, interpMode=RCIM_Cubic, leaveTangent=-0.0468, time=0.2916, value=0.709}
          {arriveTangent=-0.022, interpMode=RCIM_Cubic, leaveTangent=-0.022, time=0.4899, value=0.984}
          {arriveTangent=0.7494, interpMode=RCIM_Cubic, leaveTangent=0.7494, tangentMode=RCTM_User, time=0.7, value=0.7}
          <at 1 0>
         ]}}}
      Scale Curve=&Lerp_Float(A=0.5, B=1, Alpha=$System.OpenAmount)
     )))
   &PointForce(
    Force Strength=&RandomRangeFloat(Minimum=5, Maximum=20)
    Module.Use Falloff=false
    Module.UseLinearForce=false
   )
   &VortexForce(Vortex Axis=<1 0 0>, Vortex Force Amount=&RandomRangeFloat(Minimum=15, Maximum=50))
   &VectorNoiseForce(Force Amount=5000, Module.Use Random Noise Time Interval=true, Random Noise Time Interval=0.4) {enabled=false}
   /SetVariables(Particles.SparkleSize:Vector2f=/Niagara/DynamicInputs/Add/Add_Vector2(
     A:Vector2f=$Particles.SpriteSize
     B:Vector2f=&Vector2DFromFloat(Value=&Multiply_Float(
       A=$Particles.Sparkle
       B=&FloatFromCurve(
        FloatCurve=DataInterface() {propertyValues {Curve {keys [
            <at 0.4 0.05 RCIM_Cubic>
            {arriveTangent=-9.844, interpMode=RCIM_Cubic, leaveTangent=-9.844, tangentMode=RCTM_User, time=0.45, value=1}
            <at 0.6 0.05 RCIM_Cubic>
           ]}}}
        Scale Curve=16
       )
      ))
    )) {isSetParameters=true}
   &Drag(
    Drag=&FloatFromCurve(
     FloatCurve=DataInterface() {propertyValues {Curve {keys [
         {arriveTangent=-3.5964, interpMode=RCIM_Cubic, leaveTangent=-3.5964, tangentMode=RCTM_User, time=0, value=0.5}
         {arriveTangentWeight=0.1414, interpMode=RCIM_Cubic, leaveTangentWeight=0.33, tangentMode=RCTM_User, time=0.3, value=0.1}
         {interpMode=RCIM_Cubic, tangentMode=RCTM_User, time=1, value=1}
        ]}}}
     Scale Curve=2
    )
    Use Linear Drag=true
    Use Rotational Drag=false
   )
   &GravityForce(Gravity=<0 0 -40>) {enabled=false}
   &SolveForcesAndVelocity(Clamp Velocity=false, Limit Acceleration=false)
  ]
  rendererClasses [&NiagaraSpriteRendererProperties]
  renderers [
   {rendererClass=&NiagaraSpriteRendererProperties}
   {rendererClass=&NiagaraSpriteRendererProperties, rendererIndex=1}
  ]
 }
 Em_Portal {
  emitterUpdateScript [
   &EmitterState(Life Cycle Mode=::NewEnumerator1, Loop Behavior=::NewEnumerator1, Loop Duration Mode=::NewEnumerator1, UseLoopDelay=false)
   &SpawnBurst_Instantaneous(Use Spawn Probability=false, UseLoopCountLimit=false)
   /SetVariables(Emitter.EdgeRippleDistortion:NiagaraFloat=&FloatFromCurve(
     FloatCurve=DataInterface() {propertyValues {Curve {
        defaultValue=1
        keys [
         {interpMode=RCIM_Cubic, tangentMode=RCTM_User, time=0, value=3}
         {arriveTangent=0.038, interpMode=RCIM_Cubic, leaveTangent=0.038, tangentMode=RCTM_User, time=0.5, value=3}
         {arriveTangent=-0.1817, interpMode=RCIM_Cubic, leaveTangent=-0.1817, tangentMode=RCTM_User, time=1, value=0}
        ]
       }}}
     CurveIndex=$System.Timeline.NormalizedTime
    )) {isSetParameters=true}
  ]
  particleSpawnScript [
   &InitializeParticle(
    Color Mode=::NewEnumerator0
    Lifetime=5
    Mesh Scale Mode=::NewEnumerator3
    Mesh Uniform Scale=0.25
    Sprite Size Mode=::NewEnumerator3
    Uniform Sprite Size=250
    UsePositionOffset=false
    Write Material Random=true
   )
   &InitialMeshOrientation(Mesh Orientation Mode=::NewEnumerator2, Use Rotation Vector=false)
  ]
  particleUpdateScript [
   &ParticleState(Kill Particles When Lifetime Has Elapsed=false, Let Infinitely Lived Particles Die When Emitter Deactivates=true)
   &UpdateMeshOrientation(
    Facing Direction=&ReturnCameraVector(Camera Vectors=::NewEnumerator1, Coordinate Space=::Simulation)
    Facing Interpolation=true
    Orientation Method=::NewEnumerator1
    Override Quat=false
    TransformFacing=false
    TransformUp=false
   )
   &ScaleSpriteSize(
    Uniform Curve Index=$System.Timeline.NormalizedTime
    Uniform Curve Sprite Scale=DataInterface() {propertyValues {Curve {
       defaultValue=0
       keys [{interpMode=RCIM_Cubic, tangentMode=RCTM_User, time=0, value=0}, {interpMode=RCIM_Cubic, tangentMode=RCTM_User, time=1, value=1}]
      }}}
   )
   &ScaleMeshSize(Scale Factor=&VectorFromFloat(Value=&FloatFromCurve(
      FloatCurve=DataInterface() {propertyValues {Curve {
         defaultValue=0
         keys [
          {interpMode=RCIM_Cubic, tangentMode=RCTM_User, time=0, value=0}
          {arriveTangent=1, interpMode=RCIM_Cubic, leaveTangent=1, time=0.5, value=1.1}
          {interpMode=RCIM_Cubic, tangentMode=RCTM_User, time=1, value=1}
         ]
        }}}
      CurveIndex=$System.Timeline.NormalizedTime
     )))
   /SetVariables(Particles.Sprite Backgroudn Color:LinearColor=<rgba 0 0 0 1>) {isSetParameters=true}
  ]
  rendererClasses [&NiagaraSpriteRendererProperties, &NiagaraMeshRendererProperties]
  renderers [
   {rendererClass=&NiagaraSpriteRendererProperties}
   {rendererClass=&NiagaraMeshRendererProperties, rendererIndex=1}
  ]
 }
 Em_Portal_Flares {
  emitterUpdateScript [
   &EmitterState(
    Inactive Response=::NewEnumerator1
    Life Cycle Mode=::NewEnumerator1
    Loop Behavior=::NewEnumerator1
    Loop Duration Mode=::NewEnumerator1
    UseLoopDelay=false
   )
   &SpawnBurst_Instantaneous(Spawn Count=2, Use Spawn Probability=false)
  ]
  particleSpawnScript [
   &InitializeParticle(
    Color=<rgba 0.0935 1.0929 3.5 1>
    Sprite Size Mode=::NewEnumerator3
    Uniform Sprite Size=60
    UsePositionOffset=false
    Write Material Random=true
   )
   &CameraOffset(Camera Offset Amount=15)
  ]
  particleUpdateScript [
   &DynamicMaterialParameters(
    Index 0 Param 1=0.8
    Index 0 Param 2=0.5
    Module.Param0WriteEnabled=true
    Module.Param1WriteEnabled=true
    Module.Param2WriteEnabled=false
    Module.Param3WriteEnabled=false
   )
   &ParticleState(Kill Particles When Lifetime Has Elapsed=false)
   &ScaleSpriteSize(
    Uniform Curve Index=$System.Timeline.NormalizedTime
    Uniform Curve Sprite Scale=DataInterface() {propertyValues {Curve {
       defaultValue=1
       keys [
        {interpMode=RCIM_Cubic, tangentMode=RCTM_User, time=0, value=0}
        <at 0.2 1 RCIM_Cubic>
        {interpMode=RCIM_Cubic, tangentMode=RCTM_User, time=1, value=0}
       ]
      }}}
   )
   /SetVariables(Particles.SpriteSize2=&Multiply_Vector2(
     A=&Multiply_Vector2(A=$Particles.Initial.SpriteSize, B=<2 0.1>)
     B=&Vector2DFromFloat(Value=&FloatFromCurve(
       FloatCurve=DataInterface() {propertyValues {Curve {
          defaultValue=1
          keys [
           {interpMode=RCIM_Cubic, tangentMode=RCTM_User, time=0, value=0}
           {arriveTangent=-3.2206, interpMode=RCIM_Cubic, leaveTangent=-3.2206, tangentMode=RCTM_User, time=0.2, value=1}
           {interpMode=RCIM_Cubic, tangentMode=RCTM_User, time=1, value=0}
          ]
         }}}
       CurveIndex=$System.Timeline.NormalizedTime
       Scale Curve=2
      ))
    )) {isSetParameters=true}
   &ScaleColor(
    Scale Alpha=&FloatFromCurve(
     FloatCurve=DataInterface() {propertyValues {Curve {
        defaultValue=1
        keys [
         {interpMode=RCIM_Cubic, tangentMode=RCTM_User, time=0, value=0}
         {arriveTangent=-3.4093, interpMode=RCIM_Cubic, leaveTangent=-3.4093, tangentMode=RCTM_User, time=0.2, value=1}
         {interpMode=RCIM_Cubic, tangentMode=RCTM_User, time=1, value=0}
        ]
       }}}
     CurveIndex=$System.Timeline.NormalizedTime
     Scale Curve=5
    )
    ScaleA=true
    ScaleRGB=false
   )
  ]
  rendererClasses [&NiagaraSpriteRendererProperties]
  renderers [
   {rendererClass=&NiagaraSpriteRendererProperties}
   {rendererClass=&NiagaraSpriteRendererProperties, rendererIndex=1}
  ]
 }
 RadialFlow {
  emitterUpdateScript [
   &EmitterState(
    Inactive Response=::NewEnumerator1
    Life Cycle Mode=::NewEnumerator1
    Loop Behavior=::NewEnumerator1
    Loop Duration Mode=::NewEnumerator1
    UseLoopDelay=false
   )
   &SpawnBurst_Instantaneous(Spawn Count=4, Use Spawn Probability=false)
  ]
  particleSpawnScript [
   &InitializeParticle(
    Color=<rgba 1.635 0 3 0.5>
    Lifetime=4
    Sprite Rotation Mode=::NewEnumerator2
    Sprite Size Mode=::NewEnumerator1
    Sprite UV Mode=::NewEnumerator3
    Uniform Sprite Size Max=500
    Uniform Sprite Size Min=200
    UsePositionOffset=false
    Write Material Random=true
   )
   /SetVariables(
    Particles.Waves2Color:LinearColor=<rgba 0 1.1552 1.5 0.5>
    Particles.Waves2Offset:Vector3f=&VectorFromFloat(Value=20)
   ) {isSetParameters=true}
   /SetVariables(Particles.BiasedRandom=&FloatFromCurve(
     FloatCurve=DataInterface() {propertyValues {Curve {keys [
         <at 0 1>
         {arriveTangent=-0.4989, interpMode=RCIM_Cubic, leaveTangent=-0.4989, tangentMode=RCTM_User, time=0.4, value=0.8}
         {arriveTangent=-25.5965, interpMode=RCIM_Cubic, leaveTangent=-25.5965, tangentMode=RCTM_User, time=0.5, value=0}
         {arriveTangent=-0.5797, interpMode=RCIM_Cubic, leaveTangent=-0.5797, tangentMode=RCTM_User, time=0.6, value=-0.8}
         <at 1 -1>
        ]}}}
     CurveIndex=&RandomRangeFloat(Minimum=0, Maximum=1)
    )) {isSetParameters=true}
   &SpriteFacingAndAlignment(WriteAlignment=false, WriteFacing=true) {enabled=false}
  ]
  particleUpdateScript [
   &ParticleState(Kill Particles When Lifetime Has Elapsed=false)
   &FN_Radial_Material(
    Base Sprite Tile U:NiagaraFloat=&Lerp_Float(A=1.5, B=3, Alpha=$System.Open)
    Radial Mask Offset X=0.25
    Radial Mask Power=1.5
    Radial Mask Tile U=0.75
    Sprite Extrusion Speed=0.4
    Sprite Spin Speed:NiagaraFloat=&Multiply_Float(A=$Particles.BiasedRandom, B=0.2)
    TIme:NiagaraFloat=&Add_Float(A=&RandomRangeFloat(Minimum=0, Maximum=1), B=$System.Open)
   )
   &ScaleSpriteSize(
    Uniform Curve Index=$System.Timeline.NormalizedTime
    Uniform Curve Sprite Scale=DataInterface() {propertyValues {Curve {keys [
        {interpMode=RCIM_Cubic, tangentMode=RCTM_User, time=0, value=0}
        {arriveTangent=1.25, interpMode=RCIM_Cubic, leaveTangent=1.25, time=0.3, value=1}
        {interpMode=RCIM_Cubic, tangentMode=RCTM_User, time=0.8, value=1}
       ]}}}
   )
   /SetVariables(Particles.Waves2Size:Vector2f=&Multiply_Vector2(A=$Particles.SpriteSize, B=&Vector2DFromFloat(Value=1))) {isSetParameters=true}
   &SpriteRotationRate(Rotation Rate=&Multiply_Float(A=$Particles.BiasedRandom, B=20))
  ]
  rendererClasses [&NiagaraSpriteRendererProperties]
  renderers [
   {rendererClass=&NiagaraSpriteRendererProperties}
   {rendererClass=&NiagaraSpriteRendererProperties, rendererIndex=1}
  ]
 }
}
summary {
 systemName=NS_Utopia_Sprite_Portal
 emitters [
  {emitterName=Em_Portal, rendererClasses [&NiagaraSpriteRendererProperties, &NiagaraMeshRendererProperties]}
  {emitterName=Em_Motes, rendererClasses [&NiagaraSpriteRendererProperties]}
  {emitterName=Em_Portal_Flares, rendererClasses [&NiagaraSpriteRendererProperties]}
  {emitterName=Em_Idle_Glow, rendererClasses [&NiagaraSpriteRendererProperties]}
  {emitterName=RadialFlow, rendererClasses [&NiagaraSpriteRendererProperties]}
 ]
 userVariables [
  User.InteriorColor {type=LinearColor, description="", defaultValue {struct=LinearColor, value=<rgba 1 1 1 1>}}
  User.PortalOpen {type=NiagaraBool, defaultValue=true}
 ]
}
