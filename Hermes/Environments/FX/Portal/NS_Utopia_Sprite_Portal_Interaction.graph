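# Niagara one-shot burst effect for interacting with the Utopia sprite portal.
# A layered multi-renderer flash combines with spinning radial-material waves, flares offset by System.ImpactPos and rising light rays.
# GPU emitters add vortex-swirled motes and side sparkles whose spawn height sweeps with system age.
# System open values scale mote size and radial tiling; a longer-lived RadialFlow variant is disabled.
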NiagaraSystem /Hermes/Environments/FX/Portal/NS_Utopia_Sprite_Portal_Interaction
dependencies {
 usedDataInterfaces [
  &NiagaraDataInterfaceCurve
  &NiagaraDataInterfaceColorCurve
  &NiagaraDataInterfaceVector2DCurve
  &NiagaraDataInterfaceSpriteRendererInfo
  &NiagaraDataInterfaceVectorCurve
 ]
 usedDynamicInputs [&LerpLinearColors, &RandomRangeLinearColor]
 usedModules [&SystemState]
}
emitters {
 Em_Motes {
  simTarget=GPUComputeSim
  emitterUpdateScript [
   &EmitterState(Life Cycle Mode=::NewEnumerator1, Loop Behavior=::NewEnumerator1, Loop Duration=1, UseLoopDelay=false)
   &SpawnBurst_Instantaneous(Spawn Count=75, Use Spawn Probability=false, UseLoopCountLimit=false)
  ]
  particleSpawnScript [
   /SetVariables(Particles.SparkIntensity:NiagaraFloat=&FloatFromCurve(
     FloatCurve=DataInterface() {propertyValues {Curve {keys [
         <at 0 1 RCIM_Cubic>
         {arriveTangent=-2.6566, interpMode=RCIM_Cubic, leaveTangent=-2.6566, tangentMode=RCTM_User, time=0.3, value=0.55}
         <at 1 0 RCIM_Cubic>
        ]}}}
     CurveIndex=&RandomRangeFloat(Minimum=0, Maximum=1)
    )) {isSetParameters=true}
   &InitializeParticle(
    Color Maximum=<rgba 2.6557 11.4187 25 4>
    Color Minimum=<rgba 12.2752 1.8772 18 2>
    Color Mode=::NewEnumerator2
    Lifetime Min=0.5
    Lifetime Mode=::NewEnumerator1
    Mesh Scale Mode=::NewEnumerator4
    Position Mode=::NewEnumerator1
    Sprite Rotation Mode=::NewEnumerator2
    Sprite Size Mode=::NewEnumerator3
    Uniform Sprite Size=&Lerp_Float(A=7.5, B=15, Alpha=$Particles.SparkIntensity)
    UsePositionOffset=false
    Write Material Random=true
   )
   &SphereLocation(MaskSpawnGroup=false, Override Seed=false, Surface Only=true, UseNonUniformScale=false, UseOffset=false, UseSphereOrientation=false) {module=SphereLocation001}
   /SetVariables(Particles.Sparkle:NiagaraFloat=&UniformAOrBFloat(A=0, Distribution Weight=0.2)) {isSetParameters=true}
   &AddVelocity(Scale Added Velocity=&VectorFromFloat(Value=0)) {enabled=false}
   &AddVelocity(
    Module.Use Falloff Distance=false
    Override Seed=false
    Velocity Mode=::NewEnumerator1
    Velocity Speed=&RandomRangeFloat(Minimum=100, Maximum=500)
   ) {module=AddVelocity001}
  ]
  particleUpdateScript [
   &ParticleState()
   &DynamicMaterialParameters(
    Index 0 Param 1=1
    Index 0 Param 2=1
    Index 0 Param 3=0.5
    Module.Param0WriteEnabled=true
    Module.Param1WriteEnabled=true
    Module.Param2WriteEnabled=true
    Module.Param3WriteEnabled=true
   ) {enabled=false}
   &Drag(
    Drag=&FloatFromCurve(
     FloatCurve=DataInterface() {propertyValues {Curve {keys [
         {arriveTangent=-3.5964, interpMode=RCIM_Cubic, leaveTangent=-3.5964, tangentMode=RCTM_User, time=0, value=0.5}
         {arriveTangentWeight=0.1414, interpMode=RCIM_Cubic, leaveTangentWeight=0.33, tangentMode=RCTM_User, time=0.3, value=0.1}
         {interpMode=RCIM_Cubic, tangentMode=RCTM_User, time=1, value=1}
        ]}}}
     Scale Curve=8
    )
    Use Linear Drag=true
    Use Rotational Drag=false
   )
   &ScaleColor(
    Scale Alpha=&FloatFromCurve(FloatCurve=DataInterface() {propertyValues {Curve {keys [<at 0 0 RCIM_Cubic>, {interpMode=RCIM_Cubic, tangentMode=RCTM_User, time=0.1, value=0.9914}, <at 1 0>]}}})
    ScaleA=true
    ScaleRGB=true
   )
   &ScaleSpriteSize(Scale Factor=&Vector2DFromFloat(Value=&FloatFromCurve(
      FloatCurve=DataInterface() {propertyValues {Curve {keys [
          <at 0 0 RCIM_Cubic>
          {arriveTangent=2.4312, interpMode=RCIM_Cubic, leaveTangent=2.4312, time=0.1483, value=1}
          {arriveTangent=-0.0468, interpMode=RCIM_Cubic, leaveTangent=-0.0468, time=0.2916, value=0.709}
          {arriveTangent=-0.022, interpMode=RCIM_Cubic, leaveTangent=-0.022, time=0.4899, value=0.984}
          {arriveTangent=0.7494, interpMode=RCIM_Cubic, leaveTangent=0.7494, tangentMode=RCTM_User, time=0.7, value=0.7}
          <at 1 0>
         ]}}}
      Scale Curve=&Lerp_Float(A=0.5, B=1, Alpha=$System.OpenAmount)
     )))
   &PointForce(
    Force Strength=&RandomRangeFloat(Minimum=100, Maximum=200)
    Module.Use Falloff=false
    Module.UseLinearForce=false
   ) {enabled=false}
   &VortexForce(
    Origin Pull Amount=800
    Vortex Axis=<1 0 0>
    Vortex Force Amount=&RandomRangeFloat(Minimum=0, Maximum=200)
   )
   &VectorNoiseForce(Force Amount=5000, Module.Use Random Noise Time Interval=true, Random Noise Time Interval=0.4) {enabled=false}
   /SetVariables(Particles.SparkleSize:Vector2f=/Niagara/DynamicInputs/Add/Add_Vector2(
     A:Vector2f=$Particles.SpriteSize
     B:Vector2f=&Vector2DFromFloat(Value=&Multiply_Float(
       A=$Particles.Sparkle
       B=&FloatFromCurve(
        FloatCurve=DataInterface() {propertyValues {Curve {keys [
            <at 0.4 0.05 RCIM_Cubic>
            {arriveTangent=-9.844, interpMode=RCIM_Cubic, leaveTangent=-9.844, tangentMode=RCTM_User, time=0.45, value=1}
            <at 0.6 0.05 RCIM_Cubic>
           ]}}}
        Scale Curve=16
       )
      ))
    )) {isSetParameters=true}
   &GravityForce(Gravity=&Multiply_VectorByFloat(Vector=<0 0 -200>, Float=&RandomRangeFloat(Minimum=0, Maximum=1)))
   &PointAttractionForce(
    Attraction Radius=3500
    Attraction Strength=&FloatFromCurve(
     FloatCurve=DataInterface() {propertyValues {Curve {
        defaultValue=1
        keys [
         {arriveTangent=4.1785, interpMode=RCIM_Cubic, leaveTangent=4.1785, tangentMode=RCTM_User, time=0, value=0}
         {arriveTangent=1.1179, interpMode=RCIM_Cubic, leaveTangent=1.1179, tangentMode=RCTM_User, time=0.3, value=0.125}
         {arriveTangent=1.9643, interpMode=RCIM_Cubic, leaveTangent=1.9643, time=0.5, value=1.125}
         {arriveTangent=0.0224, interpMode=RCIM_Cubic, leaveTangent=0.0224, tangentMode=RCTM_User, time=1, value=1.5}
        ]
       }}}
     Scale Curve=100
    )
    Kill Within Radius=true
    Module.Use Falloff=true
    UseAttractorOffset=false
   ) {enabled=false}
   &SolveForcesAndVelocity(Clamp Velocity=false, Limit Acceleration=false)
  ]
  rendererClasses [&NiagaraSpriteRendererProperties]
  renderers [
   {rendererClass=&NiagaraSpriteRendererProperties}
   {rendererClass=&NiagaraSpriteRendererProperties, rendererIndex=1}
  ]
 }
 Flares {
  emitterUpdateScript [
   &EmitterState(Life Cycle Mode=::NewEnumerator1, Loop Behavior=::NewEnumerator1, Loop Duration=1.07, UseLoopDelay=true)
   &SpawnBurst_Instantaneous(Spawn Count=10, Use Spawn Probability=false, UseLoopCountLimit=false)
  ]
  particleSpawnScript [
   &InitializeParticle(
    Color=&MakeLinearColorFromVectorAndFloat("Vector (RGB)"=<1 1 1>, "Float (Alpha)"=&RandomRangeFloat(Minimum=0.35, Maximum=0.55))
    Lifetime=&RandomRangeFloat(Minimum=0.8, Maximum=2)
    Sprite Size Max=<100 150>
    Sprite Size Min=<100 75>
    Sprite Size Mode=::NewEnumerator2
    UsePositionOffset=true
    Write Material Random=true
   )
   &ShapeLocation(
    Cylinder Distribution=::NewEnumerator2
    Cylinder Height=1
    Cylinder Radius=10
    MaskSpawnGroup=false
    Offset=$System.ImpactPos
    Offset Mode=::NewEnumerator0
    Rotation Mode=::NewEnumerator1
    Shape Primitive=::NewEnumerator1
    U Position=&Lerp_Float(A=0.3, B=0.7, Alpha=&ReturnNormalizedExecIndex(Normalized Index Scale=5))
    UseOwnerScale=true
    V Position=&Lerp_Float(A=0, B=1, Alpha=&ReturnNormalizedExecIndex())
   )
   &CameraOffset(Camera Offset Amount=100)
   &SubUVAnimation(
    Evaluation Type=::NewEnumerator0
    "Sprite Renderer "=DataInterface() {propertyValues {SpriteRenderer=&self:Flares_0.NiagaraSpriteRendererProperties_2}}
    SubUV Animation Mode:ENiagara_SubUVLookupModeV2=::NewEnumerator2
    UseEndFrame=false
    UseStartFrame=false
   )
   &AddVelocity(Velocity=&NormalizeVector(Vector To Normalize=$Particles.ShapeLocation.ShapeNormal, Safe Normalize=true, Safe Fallback Threshold=0, Safe Fallback Vector=<0 0 0.001>))
  ]
  particleUpdateScript [
   &ParticleState()
   &ScaleColor(
    Scale Alpha=&FloatFromCurve(FloatCurve=DataInterface() {propertyValues {Curve {keys [
         {arriveTangent=-2.2233, interpMode=RCIM_Cubic, leaveTangent=-2.2233, tangentMode=RCTM_User, time=0.1, value=1}
         {arriveTangent=-0.0075, interpMode=RCIM_Cubic, leaveTangent=-0.0075, tangentMode=RCTM_User, time=1, value=0}
        ]}}})
    Scale RGB=&Multiply_VectorByFloat(
     Vector=&MakeVectorFromLinearColorRGB(Color=&ColorFromCurve(ColorCurve=DataInterface() {propertyValues {
         AlphaCurve {keys [<at 0 1 RCIM_Cubic>]}
         BlueCurve {keys [<at 0.0426 10 RCIM_Cubic>,<at 0.9515 4.9247>]}
         GreenCurve {keys [<at 0.0426 5.5799 RCIM_Cubic>,<at 0.9515 0>]}
         RedCurve {keys [<at 0.0426 1.562 RCIM_Cubic>,<at 0.9515 10>]}
        }}))
     Float=1
    )
    ScaleA=true
    ScaleRGB=true
   )
   &ScaleSpriteSize(
    Scale Sprite Size Mode=::NewEnumerator0
    Uniform Scale Factor=&FloatFromCurve(
     FloatCurve=DataInterface() {propertyValues {Curve {keys [
         {arriveTangent=2.0201, interpMode=RCIM_Cubic, leaveTangent=2.0201, tangentMode=RCTM_User, time=0, value=0}
         {arriveTangent=-0.026, interpMode=RCIM_Cubic, leaveTangent=-0.026, tangentMode=RCTM_User, time=0.6, value=1}
        ]}}}
     Scale Curve=3
    )
   )
   &ScaleSpriteSize(
    Non-Uniform Curve Sprite Scale=DataInterface() {propertyValues {
      XCurve {keys [
        {arriveTangent=-1.6611, interpMode=RCIM_Cubic, leaveTangent=-1.6611, tangentMode=RCTM_User, time=0.2, value=1.2}
        {arriveTangent=-0.1324, interpMode=RCIM_Cubic, leaveTangent=-0.1324, tangentMode=RCTM_User, time=1, value=0.7}
       ]}
      YCurve {keys [{interpMode=RCIM_Cubic, tangentMode=RCTM_User, time=0.4, value=1}]}
     }}
    Scale Sprite Size Mode=::NewEnumerator3
   ) {module=ScaleSpriteSize001}
   &SolveForcesAndVelocity(Clamp Velocity=false, Limit Acceleration=false)
   &DynamicMaterialParameters(
    Index 0 Param 2=1
    Index 0 Param 3=1
    Index 0 Param 4=1
    Module.Index0Param0Write=true
    Module.Index0Param1Write=true
    Module.Index0Param2Write=true
    Module.Index0Param3Write=true
   )
  ]
  rendererClasses [&NiagaraSpriteRendererProperties, &NiagaraMeshRendererProperties]
  renderers [
   {rendererClass=&NiagaraSpriteRendererProperties}
   {rendererClass=&NiagaraMeshRendererProperties, rendererIndex=1}
  ]
 }
 NE_MuzzleFlash {
  emitterUpdateScript [
   &EmitterState()
   &SpawnBurst_Instantaneous(Use Spawn Probability=false, UseLoopCountLimit=false)
  ]
  particleSpawnScript [
   &InitializeParticle(
    Color Mode=::NewEnumerator0
    Lifetime=0.85
    Mesh Scale Mode=::NewEnumerator3
    Sprite Size Mode=::NewEnumerator3
    Uniform Sprite Size=450
    UsePositionOffset=true
    Write Material Random=true
   )
   &InitialMeshOrientation(Orientation Axis=<0 0 1>, Orientation Vector=<0 0 1>, Rotation=<0 0.25 0>, Use Orientation Vector=true, Use Rotation Vector=true)
   &CameraOffset(Camera Offset Amount=250)
   &SpriteFacingAndAlignment(Sprite Alignment=<0 0 1>, WriteAlignment=true, WriteFacing=false)
  ]
  particleUpdateScript [
   &ParticleState()
   &ScaleSpriteSize(
    Uniform Curve Scale=1.2
    Uniform Curve Sprite Scale=DataInterface() {propertyValues {Curve {keys [
        {arriveTangent=2.1756, interpMode=RCIM_Cubic, leaveTangent=2.1756, tangentMode=RCTM_User, time=0, value=0}
        {interpMode=RCIM_Cubic, tangentMode=RCTM_User, time=1, value=1}
       ]}}}
   )
   &Color(Color=&ColorFromCurve(ColorCurve=DataInterface() {propertyValues {
       AlphaCurve {keys [<at 0.5878 1 RCIM_Cubic>,<at 0.9954 0>]}
       BlueCurve {keys [<at 0 4.893 RCIM_Cubic>,<at 0.6796 0.8485 RCIM_Cubic>]}
       GreenCurve {keys [<at 0 4.6667 RCIM_Cubic>,<at 0.6796 0.1778 RCIM_Cubic>]}
       RedCurve {keys [<at 0 5 RCIM_Cubic>,<at 0.6796 1 RCIM_Cubic>]}
      }}))
   /SetVariables(Particles.FlareColor=&ColorFromCurve(ColorCurve=DataInterface() {propertyValues {
       AlphaCurve {keys [<at 0 0.2 RCIM_Cubic>,<at 0.9954 0>]}
       BlueCurve {keys [<at 0 20 RCIM_Cubic>,<at 0.1501 20>,<at 0.3927 0.8485 RCIM_Cubic>]}
       GreenCurve {keys [<at 0 8.7372 RCIM_Cubic>,<at 0.1501 7.5576>,<at 0.3927 0.1778 RCIM_Cubic>]}
       RedCurve {keys [<at 0 0.6957 RCIM_Cubic>,<at 0.1501 0>,<at 0.3927 1 RCIM_Cubic>]}
      }})) {isSetParameters=true}
   /SetVariables(Particles.LongFlareSize:Vector2f=&Multiply_Vector2D(
     A=&Vector2DFromCurve(
      Vector2Curve=DataInterface() {propertyValues {XCurve {keys [<at 0 1>,<at 0.5 1>,<at 1 7>]}, YCurve {keys [<at 0 0>,<at 0.3 7>,<at 0.5 1>]}}}
      ScaleCurve=<150 150>
     )
     B=<0.7 1>
    )) {isSetParameters=true}
   &FN_Sprite_Attributes(
    Color=&MultiplyLinearColor(
     Linear Color=$Particles.Color
     Scale Factor=&MakeVector4_FromLinearColor(LinearColor=<rgba 0.0729 0.531 1 0.5>)
    )
    Create Color Attribute=true
    Create SpriteSize Attribute=true
    SpriteSize=&Multiply_Vector2D_ByFloat(Float=1, Vector2D=$Particles.SpriteSize)
    Write to Module Namespace=true
   )
   &FN_Sprite_Attributes(
    Color=&MultiplyLinearColor(
     Linear Color=$Particles.Color
     Scale Factor=&MakeVector4_FromLinearColor(LinearColor=<rgba 0.02 0.02 0.02 1>)
    )
    Create Color Attribute=true
    Create SpriteSize Attribute=true
    SpriteSize=&Multiply_Vector2D_ByFloat(Float=1.5, Vector2D=$Particles.SpriteSize)
    Write to Module Namespace=true
   ) {module=FN_Sprite_Attributes001}
   &DynamicMaterialParameters(
    Index 3 Param 1=1
    Index 3 Param 4=&Multiply_Float(A=$Engine.System.Age, B=-2)
    Module.Index3Float0Write=true
    Module.Index3Float1Write=true
    Module.Index3Float2Write=true
    Module.Index3Float3Write=true
    Write Parameter Index 0=::NewEnumerator0
    Write Parameter Index 3=::NewEnumerator1
   )
   &FN_Sprite_Attributes(
    Color=&ColorFromCurve(ColorCurve=DataInterface() {propertyValues {
       AlphaCurve {keys [<at 0 20 RCIM_Cubic>,<at 0.3345 5>,<at 0.6912 0.5>,<at 1 0 RCIM_Cubic>]}
       BlueCurve {keys [<at 0 20 RCIM_Cubic>,<at 0.3345 10 RCIM_Cubic>]}
       GreenCurve {keys [<at 0 6.7853 RCIM_Cubic>,<at 0.3345 5.5144 RCIM_Cubic>]}
       RedCurve {keys [<at 0 5.2729 RCIM_Cubic>,<at 0.3345 0.7033 RCIM_Cubic>]}
      }})
    Create Color Attribute=true
    Create SpriteSize Attribute=true
    SpriteSize=&Multiply_Vector2D_ByFloat(Float=0.7, Vector2D=$Particles.SpriteSize)
    Write to Module Namespace=true
   ) {module=FN_Sprite_Attributes002}
   &FN_Sprite_Attributes(
    Color=&MultiplyLinearColor(
     Linear Color=<rgba 0.0935 1.0929 3.5 1>
     Scale Factor=&MakeVector4FromVectorAndFloat(
      "Vector (RGB)"=&VectorFromFloat(Value=&FloatFromCurve(
        FloatCurve=DataInterface() {propertyValues {Curve {
           defaultValue=1
           keys [
            {interpMode=RCIM_Cubic, tangentMode=RCTM_User, time=0, value=0}
            {arriveTangent=-3.4093, interpMode=RCIM_Cubic, leaveTangent=-3.4093, tangentMode=RCTM_User, time=0.2, value=1}
            {interpMode=RCIM_Cubic, tangentMode=RCTM_User, time=1, value=0}
           ]
          }}}
        Scale Curve=2
       ))
      "Float (Alpha)"=&FloatFromCurve(FloatCurve=DataInterface() {propertyValues {Curve {keys [<at 0.5 1>,<at 1 0>]}}})
     )
    )
    Create Color Attribute=true
    Create SpriteSize Attribute=true
    SpriteSize=&Multiply_Vector2D_ByFloat(Float=0.7, Vector2D=$Particles.SpriteSize)
    Write to Module Namespace=true
   ) {module=FN_Sprite_Attributes003}
   /SetVariables(Particles.CoronaParams:Vector4f=<1 0.5 0 0>) {isSetParameters=true}
  ]
  rendererClasses [&NiagaraSpriteRendererProperties]
  renderers [
   {rendererClass=&NiagaraSpriteRendererProperties}
   {rendererClass=&NiagaraSpriteRendererProperties, rendererIndex=1}
   {rendererClass=&NiagaraSpriteRendererProperties, rendererIndex=2}
   {rendererClass=&NiagaraSpriteRendererProperties, rendererIndex=3}
   {rendererClass=&NiagaraSpriteRendererProperties, rendererIndex=4}
   {rendererClass=&NiagaraSpriteRendererProperties, rendererIndex=5}
  ]
 }
 RadialFlow {
  bEnabled=false
  emitterUpdateScript [
   &EmitterState(
    Inactive Response=::NewEnumerator1
    Life Cycle Mode=::NewEnumerator1
    Loop Behavior=::NewEnumerator1
    Loop Duration Mode=::NewEnumerator1
    UseLoopDelay=false
   )
   &SpawnBurst_Instantaneous(Spawn Count=4, Use Spawn Probability=false)
  ]
  particleSpawnScript [
   &InitializeParticle(
    Color=<rgba 1.635 0 3 0.5>
    Lifetime=4
    Sprite Rotation Mode=::NewEnumerator2
    Sprite Size Mode=::NewEnumerator1
    Sprite UV Mode=::NewEnumerator3
    Uniform Sprite Size Max=500
    Uniform Sprite Size Min=200
    UsePositionOffset=false
    Write Material Random=true
   )
   /SetVariables(
    Particles.Waves2Color:LinearColor=<rgba 0 1.1552 1.5 0.5>
    Particles.Waves2Offset:Vector3f=&VectorFromFloat(Value=20)
   ) {isSetParameters=true, module=SetVariables_#1}
   /SetVariables(Particles.BiasedRandom=&FloatFromCurve(
     FloatCurve=DataInterface() {propertyValues {Curve {keys [
         <at 0 1>
         {arriveTangent=-0.4989, interpMode=RCIM_Cubic, leaveTangent=-0.4989, tangentMode=RCTM_User, time=0.4, value=0.8}
         {arriveTangent=-25.5965, interpMode=RCIM_Cubic, leaveTangent=-25.5965, tangentMode=RCTM_User, time=0.5, value=0}
         {arriveTangent=-0.5797, interpMode=RCIM_Cubic, leaveTangent=-0.5797, tangentMode=RCTM_User, time=0.6, value=-0.8}
         <at 1 -1>
        ]}}}
     CurveIndex=&RandomRangeFloat(Minimum=0, Maximum=1)
    )) {isSetParameters=true, module=SetVariables_#2}
   &SpriteFacingAndAlignment(WriteAlignment=false, WriteFacing=true) {enabled=false}
  ]
  particleUpdateScript [
   &ParticleState(Kill Particles When Lifetime Has Elapsed=false)
   &FN_Radial_Material(
    Base Sprite Tile U:NiagaraFloat=&Lerp_Float(A=1.5, B=3, Alpha=$System.Open)
    Radial Mask Offset X=0.25
    Radial Mask Power=1.5
    Radial Mask Tile U=0.75
    Sprite Extrusion Speed=0.4
    Sprite Spin Speed:NiagaraFloat=&Multiply_Float(A=$Particles.BiasedRandom, B=0.2)
    TIme:NiagaraFloat=&Add_Float(A=&RandomRangeFloat(Minimum=0, Maximum=1), B=$System.Open)
   )
   &ScaleSpriteSize(
    Uniform Curve Index=$System.Timeline.NormalizedTime
    Uniform Curve Sprite Scale=DataInterface() {propertyValues {Curve {keys [
        {interpMode=RCIM_Cubic, tangentMode=RCTM_User, time=0, value=0}
        {arriveTangent=1.25, interpMode=RCIM_Cubic, leaveTangent=1.25, time=0.3, value=1}
        {interpMode=RCIM_Cubic, tangentMode=RCTM_User, time=0.8, value=1}
       ]}}}
   )
   /SetVariables(Particles.Waves2Size:Vector2f=&Multiply_Vector2(A=$Particles.SpriteSize, B=&Vector2DFromFloat(Value=1))) {isSetParameters=true, module=SetVariables_#3}
   &SpriteRotationRate(Rotation Rate=&Multiply_Float(A=$Particles.BiasedRandom, B=20))
  ]
  rendererClasses [&NiagaraSpriteRendererProperties]
  renderers [
   {rendererClass=&NiagaraSpriteRendererProperties}
   {rendererClass=&NiagaraSpriteRendererProperties, rendererIndex=1}
  ]
 }
 RadialFlow001 {
  emitterUpdateScript [
   &EmitterState()
   &SpawnBurst_Instantaneous(Spawn Count=2, Use Spawn Probability=false)
  ]
  particleSpawnScript [
   &InitializeParticle(
    Color=<rgba 1.635 0 3 0.5>
    Lifetime=1.2
    Sprite Rotation Mode=::NewEnumerator2
    Sprite Size Mode=::NewEnumerator1
    Sprite UV Mode=::NewEnumerator3
    Uniform Sprite Size Max=500
    Uniform Sprite Size Min=200
    UsePositionOffset=false
    Write Material Random=true
   )
   /SetVariables(
    Particles.Waves2Color:LinearColor=<rgba 0 1.1552 1.5 0.5>
    Particles.Waves2Offset:Vector3f=&VectorFromFloat(Value=20)
   ) {isSetParameters=true, module=SetVariables_#1}
   /SetVariables(Particles.BiasedRandom=&FloatFromCurve(
     FloatCurve=DataInterface() {propertyValues {Curve {keys [
         <at 0 1>
         {arriveTangent=-0.4989, interpMode=RCIM_Cubic, leaveTangent=-0.4989, tangentMode=RCTM_User, time=0.4, value=0.8}
         {arriveTangent=-25.5965, interpMode=RCIM_Cubic, leaveTangent=-25.5965, tangentMode=RCTM_User, time=0.5, value=0}
         {arriveTangent=-0.5797, interpMode=RCIM_Cubic, leaveTangent=-0.5797, tangentMode=RCTM_User, time=0.6, value=-0.8}
         <at 1 -1>
        ]}}}
     CurveIndex=&RandomRangeFloat(Minimum=0, Maximum=1)
    )) {isSetParameters=true, module=SetVariables_#2}
   &SpriteFacingAndAlignment(WriteAlignment=false, WriteFacing=true) {enabled=false}
  ]
  particleUpdateScript [
   &ParticleState()
   &FN_Radial_Material(
    Base Sprite Tile U:NiagaraFloat=&Lerp_Float(A=1.5, B=3, Alpha=$System.Open)
    Radial Mask Offset X=0.25
    Radial Mask Power=1.5
    Radial Mask Tile U=0.75
    Sprite Extrusion Speed=0.4
    Sprite Spin Speed:NiagaraFloat=&Multiply_Float(A=$Particles.BiasedRandom, B=0.2)
    TIme:NiagaraFloat=&Add_Float(A=&RandomRangeFloat(Minimum=0, Maximum=1), B=$System.Open)
   )
   &ScaleSpriteSize(
    Uniform Curve Scale=0.8
    Uniform Curve Sprite Scale=DataInterface() {propertyValues {Curve {keys [
        {arriveTangent=4.4328, interpMode=RCIM_Cubic, leaveTangent=4.4328, tangentMode=RCTM_User, time=0, value=0}
        {arriveTangent=0.8812, interpMode=RCIM_Cubic, leaveTangent=0.8812, tangentMode=RCTM_User, time=0.3, value=1.25}
        <at 1 1.5 RCIM_Cubic>
       ]}}}
   )
   /SetVariables(Particles.Waves2Size:Vector2f=&Multiply_Vector2(A=$Particles.SpriteSize, B=&Vector2DFromFloat(Value=1))) {isSetParameters=true, module=SetVariables_#3}
   &SpriteRotationRate(Rotation Rate=&Multiply_Float(A=$Particles.BiasedRandom, B=100))
   &ScaleColor(
    Scale Alpha=&FloatFromCurve(FloatCurve=DataInterface() {propertyValues {Curve {keys [<at 0 1 RCIM_Cubic>,<at 0.7 0 RCIM_Cubic>]}}})
    ScaleA=true
    ScaleRGB=true
   )
   /SetVariables(Particles.Waves2Color:LinearColor=&MultiplyLinearColor(
     Linear Color=$Particles.Initial.Waves2Color
     Scale Factor=&MakeVector4FromVectorAndFloat(
      "Vector (RGB)"=<1 1 1>
      "Float (Alpha)"=&FloatFromCurve(FloatCurve=DataInterface() {propertyValues {Curve {keys [<at 0.2 1 RCIM_Cubic>,<at 1 0 RCIM_Cubic>]}}})
     )
    )) {isSetParameters=true}
  ]
  rendererClasses [&NiagaraSpriteRendererProperties]
  renderers [
   {rendererClass=&NiagaraSpriteRendererProperties}
   {rendererClass=&NiagaraSpriteRendererProperties, rendererIndex=1}
  ]
 }
 Rays {
  emitterUpdateScript [
   &EmitterState(Inactive Response=::NewEnumerator1, Life Cycle Mode=::NewEnumerator1, Loop Behavior=::NewEnumerator1, Loop Duration=1, UseLoopDelay=true)
   &SpawnBurst_Instantaneous(
    Spawn Count=&RandomRangeInt(Minimum=2, Maximum=2)
    Use Spawn Probability=false
    UseLoopCountLimit=false
   )
  ]
  particleSpawnScript [
   &InitializeParticle(
    Lifetime=&RandomRangeFloat(Minimum=1, Maximum=1)
    Sprite Size=&RandomRangeVector2D(Minimum=<60 200>, Maximum=<80 240>)
    Sprite Size Mode=::NewEnumerator4
    UsePositionOffset=false
    Write Material Random=true
   )
   &ShapeLocation(
    Cylinder Height=0.1
    Cylinder Radius=20
    MaskSpawnGroup=false
    Offset=<0 0 -60>
    Offset Mode=::NewEnumerator0
    Override Seed=false
    Rotation Axis=<0 0 1>
    Rotation Mode=::NewEnumerator1
    Shape Primitive=::NewEnumerator1
    Surface Only=true
    Use Endcaps In Surface Only Mode=false
    UseOwnerScale=true
   )
   &AddVelocity(Velocity=&Multiply_VectorByFloat(
     Vector=&Multiply_Vector(A=$Particles.ShapeLocation.ShapeVector, B=<1 1 0>)
     Float=&RandomRangeFloat(Minimum=0.5, Maximum=1)
    )) {enabled=false, module=AddVelocity001}
   &AddVelocity(Velocity=&RandomRangeVector(Minimum=<-10 -10 50>, Maximum=<10 10 100>))
   &DynamicMaterialParameters(
    Index 0 Param 1=&RandomRangeFloat(Minimum=0.8, Maximum=2)
    Index 0 Param 2=&RandomRangeFloat(Minimum=0.5, Maximum=1.5)
    Index 0 Param 3=&RandomRangeFloat(Minimum=0.6, Maximum=1.6)
    Index 0 Param 4=&RandomRangeFloat(Minimum=0.5, Maximum=1.5)
    Module.Param0WriteEnabled=true
    Module.Param1WriteEnabled=true
    Module.Param2WriteEnabled=true
    Module.Param3WriteEnabled=true
   )
   &Color(Color=<rgba 20 17.4732 5.242 1>)
  ]
  particleUpdateScript [
   &ParticleState()
   &ScaleSpriteSize(
    Scale Factor=&MakeVector2D(
     X=3
     Y=&FloatFromCurve(
      FloatCurve=DataInterface() {propertyValues {Curve {keys [
          {arriveTangent=4.1785, interpMode=RCIM_Cubic, leaveTangent=4.1785, tangentMode=RCTM_User, time=0, value=0}
          {arriveTangent=1.1179, interpMode=RCIM_Cubic, leaveTangent=1.1179, tangentMode=RCTM_User, time=0.2042, value=0.6228}
          {arriveTangent=0.0224, interpMode=RCIM_Cubic, leaveTangent=0.0224, tangentMode=RCTM_User, time=1, value=1}
         ]}}}
      Scale Curve=3
     )
    )
    Scale Sprite Size Mode=::NewEnumerator2
   )
   &SolveForcesAndVelocity(Clamp Velocity=false, Limit Acceleration=false)
   &CameraOffset(Camera Offset Amount=60)
   &ScaleColor(
    Scale Alpha=&FloatFromCurve(FloatCurve=DataInterface() {propertyValues {Curve {keys [<at 0 0>,<at 0.2 1>,<at 1 0>]}}})
    Scale RGB=&VectorFromCurve(VectorCurve=DataInterface() {propertyValues {XCurve {keys [<at 0 1>,<at 1 1>]}, YCurve {keys [<at 0 1>,<at 1 1>]}, ZCurve {keys [<at 0 1>,<at 1 1>]}}})
    ScaleA=true
    ScaleRGB=true
   )
   &SpriteFacingAndAlignment(Sprite Alignment=<0 0 1>, WriteAlignment=true, WriteFacing=true)
  ]
  rendererClasses [&NiagaraSpriteRendererProperties]
  renderers [{rendererClass=&NiagaraSpriteRendererProperties}]
 }
 SparkleStationary {
  simTarget=GPUComputeSim
  emitterUpdateScript [
   &EmitterState(Life Cycle Mode=::NewEnumerator1, Loop Behavior=::NewEnumerator1, Loop Duration=0.7, UseLoopDelay=true)
   &SpawnRate(SpawnRate=100, Use Spawn Probability=false)
  ]
  particleSpawnScript [
   &InitializeParticle(
    Lifetime=&RandomRangeFloat(Minimum=0.15, Maximum=0.6)
    Sprite Size=&Vector2DFromFloat(Value=&RandomRangeFloat(Minimum=5, Maximum=9))
    Sprite Size Mode=::NewEnumerator4
    UsePositionOffset=false
    Write Material Random=true
   )
   &ShapeLocation(
    MaskSpawnGroup=false
    Offset=&MakeVector(
     X=0
     Y=&MergeHelper_FloatSwitch(0=-100, 1=100, 2=?, 3=?, Index=&RandomRangeInt(Minimum=0, Maximum=1), Num=::NewEnumerator0)
     Z=&FloatFromCurve(
      FloatCurve=DataInterface() {propertyValues {Curve {defaultValue=1, keys [{arriveTangent=4.1785, time=0, value=-1}, {arriveTangent=0.0224, leaveTangent=0.0224, time=1, value=1}]}}}
      CurveIndex=&Multiply_Float(A=$System.Age, B=3)
      Scale Curve=75
     )
    )
    Offset Mode=::NewEnumerator0
    Override Seed=false
    Rotation Axis=<0 0 1>
    Rotation Mode=::NewEnumerator1
    Sphere Radius=50
    UseOwnerScale=true
   )
   &AddVelocity(Velocity=<0 0 300>, Velocity Speed Scale=&RandomRangeFloat(Minimum=0, Maximum=1))
   &Color(Color=&MakeLinearColorFromVectorAndFloat("Vector (RGB)"=<50 50 50>, "Float (Alpha)"=&ClampFloat(Float=1)))
   &AddVelocity(
    Module.Use Falloff Distance=false
    Override Seed=false
    Velocity Mode=::NewEnumerator1
    Velocity Origin=$Particles.ShapeLocation.ShapePosition
    Velocity Speed=&RandomRangeFloat(Minimum=0, Maximum=200)
   ) {module=AddVelocity001}
   &CameraOffset(Camera Offset Amount=100)
  ]
  particleUpdateScript [
   &ParticleState()
   &ScaleSpriteSize(Uniform Curve Sprite Scale=DataInterface() {propertyValues {Curve {keys [
        {arriveTangent=9.0725, interpMode=RCIM_Cubic, time=0, value=0}
        <at 0.3 1 RCIM_Cubic>
        {interpMode=RCIM_Cubic, leaveTangent=-0.0405, time=1, value=0}
       ]}}})
   &Drag(Drag=5, Use Linear Drag=true, Use Rotational Drag=true)
   &SolveForcesAndVelocity(Clamp Velocity=false, Limit Acceleration=false)
   &ScaleColor(
    Scale RGB=&VectorFromCurve(VectorCurve=DataInterface() {propertyValues {XCurve {keys [<at 0 1>,<at 0.5 1>,<at 1 0>]}, YCurve {keys [<at 0 0>,<at 0.5 1>,<at 1 0>]}, ZCurve {keys [<at 0 1>,<at 0.5 1>,<at 1 0>]}}})
    ScaleA=true
    ScaleRGB=true
   )
  ]
  rendererClasses [&NiagaraSpriteRendererProperties]
  renderers [{rendererClass=&NiagaraSpriteRendererProperties}]
 }
}
summary {
 systemName=NS_Utopia_Sprite_Portal_Interaction
 emitters [
  {bEnabled=false, emitterName=RadialFlow, rendererClasses [&NiagaraSpriteRendererProperties]}
  {emitterName=NE_MuzzleFlash, rendererClasses [&NiagaraSpriteRendererProperties]}
  {emitterName=Flares, rendererClasses [&NiagaraSpriteRendererProperties, &NiagaraMeshRendererProperties]}
  {emitterName=RadialFlow001, rendererClasses [&NiagaraSpriteRendererProperties]}
  {emitterName=Em_Motes, simTarget=GPUComputeSim, rendererClasses [&NiagaraSpriteRendererProperties]}
  {emitterName=Rays, rendererClasses [&NiagaraSpriteRendererProperties]}
  {emitterName=SparkleStationary, simTarget=GPUComputeSim, rendererClasses [&NiagaraSpriteRendererProperties]}
 ]
 userVariables [
  User.InteriorColor {type=LinearColor, description="", defaultValue {struct=LinearColor, value=<rgba 1 1 1 1>}}
  User.PortalOpen {type=NiagaraBool, defaultValue=true}
 ]
}
